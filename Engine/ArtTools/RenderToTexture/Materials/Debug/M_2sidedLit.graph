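# M_2sidedLit is a minimal two-sided, Default Lit debug material in the engine's RenderToTexture art tools.
# Its graph holds only a single mid-value scalar constant, with no connection to a material input shown.
# It has no textures or parameters, and pixel depth offset uses the legacy mode.
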Material /Engine/ArtTools/RenderToTexture/Materials/Debug/M_2sidedLit
graph {properties {pixelDepthOffsetMode=PDOM_Legacy, twoSided=true, usedShadingModels=Default Lit}, nodes [Constant() {r=0.5}]}
