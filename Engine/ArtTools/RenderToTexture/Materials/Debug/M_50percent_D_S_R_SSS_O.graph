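# Debug material in the RenderToTexture art tools, named as a 50 percent test case for what appear to be diffuse, specular, roughness, subsurface and opacity channels.
# It uses the Subsurface shading model, and its graph holds a single mid-gray scalar constant with no textures or parameters.
# The name suggests that constant drives those channels, though the graph shows no input connections.
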Material /Engine/ArtTools/RenderToTexture/Materials/Debug/M_50percent_D_S_R_SSS_O
graph {properties {pixelDepthOffsetMode=PDOM_Legacy, shadingModel=MSM_Subsurface, usedShadingModels=Subsurface}, nodes [Constant() {r=0.5}]}
