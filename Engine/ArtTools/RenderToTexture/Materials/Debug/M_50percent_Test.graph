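# M_50percent_Test is an unlit debug material in the RenderToTexture art tools, named as a 50 percent test.
# Its graph holds a single Constant node of 0.5, a mid-level value, with no textures, parameters or math.
# It is flagged for use with static lighting, and the listing shows no link from the constant to a material output.
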Material /Engine/ArtTools/RenderToTexture/Materials/Debug/M_50percent_Test
graph {
 properties {bUsedWithStaticLighting=true, pixelDepthOffsetMode=PDOM_Legacy, shadingModel=MSM_Unlit, usedShadingModels=Unlit}
 nodes [Constant() {r=0.5}]
}
