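# Debug material in the RenderToTexture art tools, named for a constant base color, using the Default Lit shading model.
# The graph holds only a Constant node and a scalar parameter named Value, with no links shown from either to a material output.
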Material /Engine/ArtTools/RenderToTexture/Materials/Debug/M_BaseColor_Constant
graph {
 properties {pixelDepthOffsetMode=PDOM_Legacy, usedShadingModels=Default Lit}
 nodes [
  Constant()
  ScalarParameter() {defaultValue=0.5, parameterName=Value}
 ]
}
