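# Unlit debug material in the RenderToTexture art tools, named as a solid emissive color.
# Its only node is a vector parameter called Color, exposed for material instances to override, with no link to an output shown.
# The material is flagged for use with static lighting.
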Material /Engine/ArtTools/RenderToTexture/Materials/Debug/M_Emissive_Color
graph {
 properties {bUsedWithStaticLighting=true, pixelDepthOffsetMode=PDOM_Legacy, shadingModel=MSM_Unlit, usedShadingModels=Unlit}
 nodes [VectorParameter() {parameterName=Color}]
}
