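# Debug material in the engine's RenderToTexture art tools, named as a constant emissive.
# It uses the Unlit shading model and holds a single scalar parameter, Value, defaulting to mid-level.
# The graph contains no other nodes and shows no link from that parameter to a material output.
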Material /Engine/ArtTools/RenderToTexture/Materials/Debug/M_Emissive_Constant
graph {
 properties {pixelDepthOffsetMode=PDOM_Legacy, shadingModel=MSM_Unlit, usedShadingModels=Unlit}
 nodes [ScalarParameter() {defaultValue=0.5, parameterName=Value}]
}
