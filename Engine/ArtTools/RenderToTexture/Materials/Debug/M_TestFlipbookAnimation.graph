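# Debug material in the RenderToTexture art tools that builds an animatable concentric-ring pattern, apparently for testing flipbook animation.
# UV distance from the center is scaled and offset by the Animation Phase parameter, then Frac, Multiply and Floor turn it into hard binary bands.
# A divide of that distance by a Size parameter and an empty Subtract feed nothing, and no link to a material output is shown.
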Material /Engine/ArtTools/RenderToTexture/Materials/Debug/M_TestFlipbookAnimation
graph {
 properties {pixelDepthOffsetMode=PDOM_Legacy, usedShadingModels=Default Lit}
 nodes [
  n1 = Distance(A=TextureCoordinate(), B=Constant() {r=0.5})
  Divide(A=@n1, B=ScalarParameter() {defaultValue=512, parameterName=Size})
  n2 = Add(A=Multiply(A=@n1) {constB=10}, B=ScalarParameter() {parameterName=Animation Phase})
  Floor(None=Multiply(A=Frac(None=@n2)) {constB=2})
  Subtract()
 ]
}
